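# Concrete slab material for DriveIn environment props, with slab and damaged-edge variants chosen by the IsEdgeMask static switch.
# Slabs layer two tilings of concrete with dirt and cracks, using vertex-color height lerps for cracks (red), tone variation (green) and landscape-matched grass on upward faces (blue).
# The edge variant blends in edge textures, writes their alpha to opacity mask, and offsets vertices toward the camera by distance and along normals.
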Material /Game/Environments/Props/DriveIn/Materials/M_ConcreteSlabCustom
graph {
 properties {bUseMaterialAttributes=true, bUsedWithNanite=true, pixelDepthOffsetMode=PDOM_Legacy, usedShadingModels=Default Lit}
 nodes [
  Divide()
  MaterialFunctionCall() {materialFunction=/Game/Athena/Environments/Landscape/MaterialFunctions/Standard/MF_Athena_Grass_01}
  n1 = Clamp(None=Add(
    A=ScalarParameter() {defaultValue=0.65, parameterName=Darkenedges}
    B=TextureSample() {texture=/Game/Environments/Props/DriveIn/Textures/T_FilmLensDirt}.G
   ))
  n2 = Multiply(A=TextureCoordinate(), B=ScalarParameter() {defaultValue=0.9695, parameterName=DirtTiling})
  n3 = Multiply(A=TextureCoordinate(), B=ScalarParameter() {defaultValue=19.0536, parameterName=ConcreteTiling})
  n4 = Multiply(A=TextureCoordinate(), B=ScalarParameter() {defaultValue=3, parameterName=ConcreteTiling_2})
  n5 = TextureSample(UVs=@n4) {texture=/Game/Environments/Props/DriveIn/Textures/T_concrete_slab_01_D}
  n6 = Multiply(A=@n5.R, B=ScalarParameter() {defaultValue=2.6449, parameterName=2ndLayerBrightness})
  n7 = Multiply(A=TextureCoordinate(), B=ScalarParameter() {defaultValue=29.9405, parameterName=CracksTiling})
  n8 = TextureSample(UVs=@n7) {texture=/Game/Environments/Props/DriveIn/Textures/T_concrete_slab_cracks_01_D}
  n9 = Add(A=ScalarParameter() {defaultValue=0.581, parameterName=CrackIntensity}, B=@n8.A)
  n10 = MaterialFunctionCall(Transition Phase=OneMinus(None=VertexColor().R), Height Texture=@n8.A, Contrast=ScalarParameter() {parameterName=Contrast}) {materialFunction=&HeightLerp}
  n11 = Multiply(
   A=Multiply(A=TextureSample(UVs=@n3) {texture=&T_concrete_slab_01_D}.RGB, B=@n6)
   B=LinearInterpolate(A=Constant3Vector() {constant=<rgba 1 1 1 0>}, B=Clamp(None=@n9), Alpha=@n10.Alpha)
  )
  n12 = Multiply(
   A=LinearInterpolate(
    A=Constant3Vector() {constant=<rgba 1 1 1 0>}
    B=VectorParameter() {parameterName=ColorDirt, defaultValue=<rgba 0 0 0 1>}.RGB
    Alpha=TextureSample(UVs=@n2) {texture=&T_FilmLensDirt}.R
   )
   B=@n11
  )
  n13 = TextureSample(UVs=@n3) {samplerType=SAMPLERTYPE_Masks, texture=/Game/Environments/Props/DriveIn/Textures/T_concrete_slab_01_S}
  n14 = TextureSample(UVs=@n3) {samplerType=SAMPLERTYPE_Normal, texture=/Game/Environments/Props/DriveIn/Textures/T_concrete_slab_01_N}
  n15 = TextureSample(UVs=@n4) {samplerType=SAMPLERTYPE_Normal, texture=&T_concrete_slab_01_N}
  n16 = Constant3Vector() {constant=<rgba 0 0 1 0>}
  n17 = TextureSample(UVs=@n7) {samplerType=SAMPLERTYPE_Normal, texture=/Game/Environments/Props/DriveIn/Textures/T_concrete_slab_cracks_01_N}
  n18 = MakeMaterialAttributes!(
   BaseColor=ShadingPathSwitch(Default=Multiply(A=@n1, B=@n12), Mobile=@n11) {desc=Migrated from FeatureLevelSwitch}
   Metallic=@n13.G
   Specular=@n13.G
   Roughness=@n13.G
   Anisotropy
   EmissiveColor
   Opacity
   OpacityMask
   Normal=MaterialFunctionCall(
    BaseNormal=MaterialFunctionCall(BaseNormal=@n14.RGB, AdditionalNormal=@n15.RGB) {materialFunction=&BlendAngleCorrectedNormals}.Result
    AdditionalNormal=LinearInterpolate(A=@n16, B=@n17.RGB, Alpha=@n10.Alpha)
   ) {materialFunction=&BlendAngleCorrectedNormals}.Result
   Tangent
   WorldPositionOffset
   SubsurfaceColor
   ClearCoat
   ClearCoatRoughness
   AmbientOcclusion
   Refraction
   CustomizedUVs_0
   CustomizedUVs_1
   CustomizedUVs_2
   CustomizedUVs_3
   CustomizedUVs_4
   CustomizedUVs_5
   CustomizedUVs_6
   CustomizedUVs_7
   PixelDepthOffset
   ShadingModel
   Displacement
  )
  n19 = TextureSample() {texture=/Game/Environments/Props/DriveIn/Textures/T_concrete_slab_edges_D}
  n20 = TextureSample() {samplerType=SAMPLERTYPE_Masks, texture=/Game/Environments/Props/DriveIn/Textures/T_concrete_slab_edges_S}
  n21 = TextureSample() {samplerType=SAMPLERTYPE_Normal, texture=/Game/Environments/Props/DriveIn/Textures/T_concrete_slab_edges_N}
  n22 = MaterialFunctionCall(Normal=@n15.RGB, Flatness=ScalarParameter() {parameterName=EdgeDetailNormalFlatten}) {materialFunction=&FlattenNormal}
  n23 = MakeMaterialAttributes(
   BaseColor=Multiply(A=VectorParameter() {parameterName=ColorEdgeDamage, defaultValue=<rgba 0.7708 0.7708 0.7708 1>}.RGB, B=Multiply(A=@n6, B=@n19.RGB))
   Metallic=@n20.G
   Specular=@n20.G
   Roughness=@n20.G
   Normal=ShadingPathSwitch(
    Default=MaterialFunctionCall(BaseNormal=@n21.RGB, AdditionalNormal=@n22.Result) {materialFunction=&BlendAngleCorrectedNormals}.Result
    Mobile=@n21.RGB
   ) {desc=Migrated from FeatureLevelSwitch}
  )
  n24 = MaterialFunctionCall!(Base Material=@n18, Top Material=@n23, Alpha=@n19.A) {materialFunction=/Engine/Functions/MaterialLayerFunctions/MatLayerBlend_Standard}
  n25 = CameraPositionWS()
  n26 = WorldPosition()
  n27 = ScalarParameter() {defaultValue=8000, parameterName=Distance}
  n28 = Multiply(
   A=Normalize(VectorInput=Subtract(A=@n25, B=@n26.XYZ))
   B=LinearInterpolate(
    A=ScalarParameter() {defaultValue=2, parameterName=OffsetMin}
    B=ScalarParameter() {defaultValue=50.5746, parameterName=OffsetMax}
    Alpha=MaterialFunctionCall(Distance=@n27) {materialFunction=&Near_Camera_Fade}.Result
   )
  )
  n29 = Multiply(A=VertexNormalWS(), B=ScalarParameter() {defaultValue=2, parameterName=OffsetMin})
  n30 = GetMaterialAttributes(None=@n18) {attributeGetTypes [#1]}
  n31 = MaterialFunctionCall(
   A=@n30.Base Color
   B=Multiply(A=@n30.Base Color, B=VectorParameter() {parameterName=ToneColor, defaultValue=<rgba 0.5938 0.5938 0.5938 0>}.RGB)
   Transition Phase=VertexColor().G
   Height Texture=@n5.G
   Contrast=ScalarParameter() {parameterName=ToneContrast}
  ) {materialFunction=&HeightLerp}
  n32 = MaterialFunctionCall!(Water, GrassColor) {materialFunction=/Game/Athena/Environments/Landscape/MaterialFunctions/Standard/MF_Athena_LakeGrass_01}
  n33 = MaterialFunctionCall!(UseVertexInterp) {materialFunction=/Game/Athena/Environments/Landscape/MaterialFunctions/MF_Landscape_GrassDistanceBlend}
  n34 = MaterialFunctionCall!(BaseColor=StaticSwitchParameter(True=GetMaterialAttributes(None=@n32.Result) {attributeGetTypes [#1]}.Base Color, False=@n33.Grass RGB) {parameterName=ForceLakeGrass}) {materialFunction=/Game/Athena/Environments/Landscape/MaterialFunctions/Standard/MF_Athena_GrassMatch}
  n35 = ComponentMask(None=AppendVector(A=@n33.Grass RGB, B=@n33.Grass RGB)) {a=true}
  n36 = LinearInterpolate(
   B=ScalarParameter() {defaultValue=0.15, group=Grass, parameterName=GrassSpecAmount}
   Alpha=Power(Base=@n35, Exp=ScalarParameter() {defaultValue=2, group=Grass, parameterName=GrassSpecPower})
  )
  n37 = Constant() {r=0.9}
  n38 = OneMinus(None=@n35)
  n39 = Multiply(A=VertexColor().B, B=DotProduct(A=Constant3Vector() {constant=<rgba 0 0 1 0>}, B=VertexNormalWS()))
  n40 = ScalarParameter() {defaultValue=0.3769, parameterName=Contrast_GrassTransition}
  StaticSwitchParameter(
   True=SetMaterialAttributes!(
    MaterialAttributes=ShadingPathSwitch(Default=@n24.Blended Material, Mobile=@n23) {desc=Migrated from FeatureLevelSwitch}
    Opacity Mask=@n19.A
    World Position Offset=Add(A=@n28, B=@n29)
   ) {attributeSetTypes [#2, #3]}
   False=MaterialFunctionCall(
    Base Material=SetMaterialAttributes!(MaterialAttributes=@n30.MaterialAttributes, Base Color=@n31.Results) {attributeSetTypes [#1]}
    Top Material=MakeMaterialAttributes(BaseColor=@n34.Result, Specular=@n36, Roughness=@n37, Opacity=@n38)
    Alpha=MaterialFunctionCall(Transition Phase=@n39, Height Texture=@n33.Grass A, Contrast=@n40) {materialFunction=&HeightLerp}.Alpha
   ) {materialFunction=&MatLayerBlend_Standard}.Blended Material
  ) {parameterName=IsEdgeMask}
 ]
}
